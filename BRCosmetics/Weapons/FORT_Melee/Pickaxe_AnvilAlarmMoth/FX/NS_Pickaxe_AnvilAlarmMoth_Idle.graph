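# Niagara idle effect for the AnvilAlarmMoth pickaxe, layering embers, mesh fire and soft glow sprites on weapon sockets.
# Fire meshes and glow sprites use particle partitions to pin each particle to a socket with its own offset and shape.
# Event audio amplitude triggers rising ember bursts.
# A smoothed system value, MusicReactivity and User.AlwaysOnToggle drive fire and glow opacity, with front-end variants for camera offset and glow alignment.
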NiagaraSystem /BRCosmetics/Weapons/FORT_Melee/Pickaxe_AnvilAlarmMoth/FX/NS_Pickaxe_AnvilAlarmMoth_Idle
dependencies {
 usedDataInterfaces [
  &NiagaraDataInterfaceArrayFloat
  &NiagaraDataInterfaceArrayInt32
  &NiagaraDataInterfaceArrayFloat2
  &NiagaraDataInterfaceCurve
  &NiagaraDataInterfaceSpriteRendererInfo
  &NiagaraDataInterfaceColorCurve
  &NiagaraDataInterfaceArrayFloat3
  &NiagaraDataInterfaceSocketReader
 ]
 usedDynamicInputs [&VectorLength]
 usedModules [
  &self:SystemSpawnScript.SystemScriptSource.SystemScriptGraph.NiagaraNodeAssignment_1.SetVariables_#1
  &self:SystemSpawnScript.SystemScriptSource.SystemScriptGraph.NiagaraNodeAssignment_2.SetVariables_#2
  &SystemState
  &InterpolateOverTime
  &self:SystemSpawnScript.SystemScriptSource.SystemScriptGraph.NiagaraNodeAssignment_0.SetVariables_#3
  &self:SystemSpawnScript.SystemScriptSource.SystemScriptGraph.NiagaraNodeAssignment_3.SetVariables_#4
 ]
}
emitters {
 NE_Embers {
  emitterUpdateScript [
   &EmitterState(Life Cycle Mode=::NewEnumerator1, Loop Behavior=::NewEnumerator1, Loop Delay=1, Loop Duration Mode=::NewEnumerator1, UseLoopDelay=false)
   /Game/Effects/Niagara/Modules/Spawn/FN_Spawn_TriggeredBurst(
    Cooldown=0.5
    Module.Use Burst Count Limit=false
    Spawn:NiagaraBool=&SetBoolByFloatComparison(A=$NPC.FortniteNPC.EventAudio_Amplitude, B=0.85)
    Spawn Count=5
    UseLoopCountLimit=false
   )
  ]
  particleSpawnScript [
   &InitializeParticle(
    Color=&FNTimeOfDayColor(Night Color=<rgba 0.5 0.5 0.5 0.1>, Morning Color=<rgba 1 1 1 0.15>, Day Color=<rgba 1 1 1 0.25>, Evening Color=<rgba 1 1 1 0.15>)
    Lifetime Max=0.6
    Lifetime Min=0.2
    Lifetime Mode=::NewEnumerator1
    Position Offset=<0 0 30>
    Position Offset Coordinate Space=::Local
    Sprite Rotation Mode=::NewEnumerator2
    Sprite Size Mode=::NewEnumerator1
    Uniform Sprite Size Max=15
    Uniform Sprite Size Min=7.5
    UsePositionOffset=true
    Write Material Random=true
   )
   &ShapeLocation(MaskSpawnGroup=false, Override Seed=false, Sphere Radius=10, UseOwnerScale=true)
   &SocketLocation(
    OverrideRandomSeed=false
    SamplePosOffset=false
    Socket Reader=$System.Socket Reader
    Socket Sampling Mode=::NewEnumerator1
    VelocitySpeedLimit=false
   )
  ]
  particleUpdateScript [
   &ParticleState()
   &ScaleSpriteSize(Uniform Curve Sprite Scale=DataInterface() {propertyValues {Curve {keys [
        {arriveTangent=6.7193, interpMode=RCIM_Cubic, leaveTangent=6.7193, tangentMode=RCTM_User, time=0, value=0}
        {arriveTangent=0.3376, interpMode=RCIM_Cubic, leaveTangent=-4.7872, tangentMode=RCTM_Break, time=0.4001, value=1.0058}
        <at 1 0>
       ]}}})
   &SpriteRotationRate(Rotation Rate=&AngleConversion(
     Angle:NiagaraFloat=&RandomRangeFloat(Minimum=-1, Maximum=1)
     Angle Input:ENiagara_AngleInput=::NewEnumerator1
     Angle Output:ENiagara_AngleInput=::NewEnumerator0
    ))
   &Drag(Drag=&RandomRangeFloat(Minimum=6, Maximum=8), Use Linear Drag=true, Use Rotational Drag=true)
   &AccelerationForce(Acceleration=&RandomRangeVector(Minimum=<0 0 500>, Maximum=<0 0 750>))
   &InheritSourceMovement(ApplyForce=false, ApplyPosition=true, ApplyPositionFalloff=true, ApplyPositionFalloffCurve=false)
   &SolveForcesAndVelocity(Clamp Velocity=false, Limit Acceleration=false)
   &SubUVAnimation(
    Random Start Frame=true
    "Sprite Renderer "=DataInterface() {propertyValues {SpriteRenderer=&self:NE_Embers_0.NiagaraSpriteRendererProperties_0}}
    SubUV Animation Mode:ENiagara_SubUVLookupModeV2=::NewEnumerator3
    UseEndFrame=false
    UsePlayRateScale=false
    UseStartFrame=false
   )
   &ScaleColor(
    Scale Alpha=&FloatFromCurve(
     FloatCurve=DataInterface() {propertyValues {Curve {keys [
         {interpMode=RCIM_Cubic, tangentMode=RCTM_User, time=0, value=0}
         {interpMode=RCIM_Cubic, tangentMode=RCTM_User, time=0.2, value=1}
         {interpMode=RCIM_Cubic, tangentMode=RCTM_User, time=1, value=0}
        ]}}}
     Scale Curve=&MakeCustomFloatFromBool(
      Bool=&SetBoolByFloatComparison(A=$User.AlwaysOnToggle, B=2, Comparison Type=::NewEnumerator4)
      True Float=0
      False Float=1
     )
    )
    Scale RGB=&VectorFromFloat(Value=&FloatFromCurve(
      FloatCurve=DataInterface() {propertyValues {Curve {
         defaultValue=1
         keys [{interpMode=RCIM_Cubic, tangentMode=RCTM_User, time=0, value=1}, {interpMode=RCIM_Cubic, tangentMode=RCTM_User, time=1, value=0}]
        }}}
      Scale Curve=2
     ))
    ScaleA=true
    ScaleRGB=true
   )
  ]
  rendererClasses [&NiagaraSpriteRendererProperties]
  renderers [{rendererClass=&NiagaraSpriteRendererProperties}]
 }
 NE_Fire_Additional {
  emitterUpdateScript [
   &EmitterState(
    Inactive Response=::NewEnumerator1
    Life Cycle Mode=::NewEnumerator1
    Loop Behavior=::NewEnumerator1
    Loop Delay=0.05
    Loop Duration Mode=::NewEnumerator1
    UseLoopDelay=false
   )
   &SpawnBurst_Instantaneous(Spawn Count=10, Use Spawn Probability=false, UseLoopCountLimit=true)
  ]
  particleSpawnScript [
   &PartitionParticles(Number Of Partitions=10)
   &InitializeParticle(
    Color=<rgba 1 1 1 0.2>
    Lifetime=6
    Mesh Scale=&Multiply_Vector(
     A=<0.15 0.8 0.1>
     B=&MergeHelper_VectorSwitch(
      0=<0.75 1 0.5>
      1=<0.65 1 1>
      2=<0.5 1 1>
      3=<0.65 1 1.5>
      4=<1 1 2>
      5=<0.65 1 1.5>
      6=<0.55 1 1.5>
      7=<0.65 1 1>
      8=<0.3 1 2>
      9=<1 1 1>
      Index=$Particles.PartitionParticles.Partition
      Num=::NewEnumerator8
     )
    )
    Mesh Scale Mode=::NewEnumerator4
    UsePositionOffset=false
    Write Material Random=true
   )
   &InitialMeshOrientation(
    Mesh Axis To Orient=<0 1 0>
    Mesh Orientation Mode=::NewEnumerator2
    Rotation=&MergeHelper_VectorSwitch(
     0=<0 0 0>
     1=<0 0 0>
     2=<0 -0.1 0>
     3=<0 -0.1 0>
     4=<0 -0.1 0>
     5=<0 -0.1 0>
     6=<0 -0.1 0>
     7=<0 -0.3 0>
     8=<0 -0.3 0>
     9=<0 -0.15 0>
     Index=$Particles.PartitionParticles.Partition
     Num=::NewEnumerator8
    )
    Use Rotation Vector=true
   )
   &CameraOffset(Camera Offset Amount=&FN_FrontEndFloatModifier(Gameplay=-5, Front End=-6))
  ]
  particleUpdateScript [
   &ParticleState(Kill Particles When Lifetime Has Elapsed=false)
   &SocketLocation(
    SamplePosOffset=true
    Sampled Position Offset=&MergeHelper_VectorSwitch(
     0=<0 0 0>
     1=<0 0 0>
     2=<0 0 0>
     3=<0 0 -2>
     4=<0 0 -0.5>
     5=<0 0 -1>
     6=<0 0 -2>
     7=<0 -5 -3>
     8=<0 -2 0>
     9=<0 0 -4>
     Index=$Particles.PartitionParticles.Partition
     Num=::NewEnumerator8
    )
    Socket Index=$Particles.PartitionParticles.Partition
    Socket Reader=$System.Socket Reader
    Socket Sampling Mode=::NewEnumerator3
    VelocitySpeedLimit=false
   )
   &ScaleColor(
    Scale Alpha=&MakeCustomFloatFromBool(
     Bool=&SetBoolByFloatComparison(A=$User.AlwaysOnToggle, B=2, Comparison Type=::NewEnumerator4)
     True Float=0
     False Float=&Lerp_Float(
      A=&Lerp_Float(
       A=&Multiply_Float(
        A=&Lerp_Float(A=0, B=1, Alpha=$System.InterpolateOverTime002.InterpolateOverTime_Float.Moving Average)
        B=&ClampFloat(Float=$NPC.FortniteNPC.EventAudio_Amplitude, Min=0.5, Max=0.5)
       )
       B=1
       Alpha=$System.MusicReactivity
      )
      B=0.5
      Alpha=$User.AlwaysOnToggle
     )
    )
    ScaleA=true
    ScaleRGB=true
   )
   &DynamicMaterialParameters(
    Index 0 Param 1=1
    Module.Index0Float0Write=true
    Module.Index0Float1Write=true
    Module.Index0Float2Write=true
    Module.Index0Float3Write=true
   )
  ]
  rendererClasses [&NiagaraMeshRendererProperties]
  renderers [
   {rendererClass=&NiagaraMeshRendererProperties, rendererIndex=0}
   {rendererClass=&NiagaraMeshRendererProperties, rendererIndex=1}
   {rendererClass=&NiagaraMeshRendererProperties, rendererIndex=2}
   {rendererClass=&NiagaraMeshRendererProperties, rendererIndex=3}
  ]
 }
 NE_Glow {
  emitterUpdateScript [
   &EmitterState(
    Inactive Response=::NewEnumerator1
    Life Cycle Mode=::NewEnumerator1
    Loop Behavior=::NewEnumerator1
    Loop Duration Mode=::NewEnumerator1
    UseLoopDelay=false
   )
   &SpawnBurst_Instantaneous(Spawn Count=5, Use Spawn Probability=false, UseLoopCountLimit=false)
  ]
  particleSpawnScript [
   &PartitionParticles(Number Of Partitions=5)
   &InitializeParticle(
    Color=<rgba 1 1 1 0.25>
    Sprite Size=&MergeHelper_Vector2Switch(0=<50 95>, 1=<55 95>, 2=<55 95>, 3=<55 95>, 4=<55 95>, Index=$Particles.PartitionParticles.Partition, Num=::NewEnumerator3)
    Sprite Size Mode=::NewEnumerator4
    UsePositionOffset=false
    Write Material Random=true
   )
   &CameraOffset()
  ]
  particleUpdateScript [
   &ParticleState(Kill Particles When Lifetime Has Elapsed=false, Let Infinitely Lived Particles Die When Emitter Deactivates=true)
   &SocketLocation(
    SamplePosOffset=true
    Sampled Position Offset=&SelectVectorFromArray(
     Direct Array Index=$Particles.PartitionParticles.Partition
     Vector Selection Array=DataInterface() {propertyValues {FloatData [<0 -3 0>, <0 0 0>, <0 0 0>, <0 0 0>, <0 0 0>]}}
    )
    Socket Index=&SelectIntFromArray(Direct Array Index=$Particles.PartitionParticles.Partition, Int Selection Array=DataInterface() {propertyValues {IntData [0,2,4,6,8]}})
    Socket Reader=$System.Socket Reader
    Socket Sampling Mode=::NewEnumerator3
    VelocitySpeedLimit=false
   )
   &SpriteFacingAndAlignment(
    Sprite Alignment=&FN_FrontEndVector3Modifier(
     Gameplay=&SelectVectorFromArray(
      Direct Array Index=$Particles.PartitionParticles.Partition
      Vector Selection Array=DataInterface() {propertyValues {FloatData [<0 0 0>, <0 1 1>, <0 0 1>, <-1 0 1>, <-1 0 1>]}}
     )
     Front End=&SelectVectorFromArray(
      Direct Array Index=$Particles.PartitionParticles.Partition
      Vector Selection Array=DataInterface() {propertyValues {FloatData [<-1 0 0>, <0 1 1>, <0 0 1>, <-1 0 1>, <-1 0 1>]}}
     )
    )
    WriteAlignment=true
    WriteFacing=false
   )
   &ScaleColor(Color Scale:LinearColor=<rgba 1 1 1 0.1>, Scale Mode=3)
   &ScaleColor(
    Scale Alpha=&MakeCustomFloatFromBool(
     Bool=&SetBoolByFloatComparison(A=$User.AlwaysOnToggle, B=2, Comparison Type=::NewEnumerator4)
     True Float=0
     False Float=&Lerp_Float(
      A=&Lerp_Float(
       A=&Lerp_Float(
        A=0
        B=&FNTimeOfDayScalar(Night Scale=0.2, Morning Scale=0.2, Day Scale=0.2, Evening Scale=0.2)
        Alpha=$System.InterpolateOverTime002.InterpolateOverTime_Float.Moving Average
       )
       B=0.1
       Alpha=$System.MusicReactivity
      )
      B=0.2
      Alpha=$User.AlwaysOnToggle
     )
    )
    ScaleA=true
    ScaleRGB=true
   ) {module=ScaleColor001}
  ]
  rendererClasses [&NiagaraSpriteRendererProperties]
  renderers [{rendererClass=&NiagaraSpriteRendererProperties}]
 }
}
summary {
 systemName=NS_Pickaxe_AnvilAlarmMoth_Idle
 emitters [
  {emitterName=NE_Embers, rendererClasses [&NiagaraSpriteRendererProperties]}
  {emitterName=NE_Glow, rendererClasses [&NiagaraSpriteRendererProperties]}
  {emitterName=NE_Fire_Additional, rendererClasses [&NiagaraMeshRendererProperties]}
 ]
 userVariables [
  User.AlwaysOnToggle {type=NiagaraFloat, defaultValue=0, description=""}
  User.bIsFrontEnd {type=NiagaraBool}
  User.bIsFrontEndPreview {type=NiagaraBool}
  User.Emote begin {type=NiagaraBool, defaultValue=true}
  User.MusicFE {type=NiagaraFloat, defaultValue=0, description=""}
 ]
}
